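# M_Darkshadow is an unlit translucent particle material for the Space Bunny pickaxe FX, drawing a soft, noisy, tinted blob.
# Opacity multiplies panning screen-aligned noise by a curve-atlas-shaped radial mask, then applies depth fade and optional particle alpha.
# A camera-axis offset built from UVs and vertex-red-driven size forms a billboard, apparently as World Position Offset, with a lerp to disable it.
# Color, optionally tinted by particle color, passes through MF_FX_EmissiveAdjustments.
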Material /BRCosmetics/Weapons/FORT_Melee/Pickaxe_Space_Bunny/FX/M_Darkshadow
graph {
 properties {
  bUsedWithNiagaraSprites=true
  bUsedWithParticleSprites=true
  blendMode=BLEND_Translucent
  pixelDepthOffsetMode=PDOM_Legacy
  shadingModel=MSM_Unlit
  translucencyPass=MTP_BeforeDOF
  usedShadingModels=Unlit
 }
 nodes [
  n1 = MaterialFunctionCall!(TextureResolution=ScalarParameter() {defaultValue=512, parameterName=512}) {materialFunction=/Engine/Functions/Engine_MaterialFunctions02/Texturing/ScreenAlignedPixelToPixelUVs}
  n2 = AppendVector(A=ScalarParameter() {parameterName=SpeedX}, B=ScalarParameter() {defaultValue=0.05, parameterName=SpeedY})
  n3 = LinearInterpolate(Alpha=TextureSample(UVs=Panner(Coordinate=@n1.Result, Speed=@n2)) {samplerType=SAMPLERTYPE_LinearColor, texture=&T_NoisePack_LinearColor}.R) {constA=0.5}
  n4 = CurveAtlasRowParameter(CurveTime=SphereMask(A=TextureCoordinate(), B=Constant() {r=0.5}, Radius=ScalarParameter() {defaultValue=0.5, parameterName=MaskRadius}, Hardness=Constant())) {
   atlas=&AthenaCurveAtlas01
   curve=/Game/Athena/LinearColorCurves/Curve_Masks2
   defaultValue=69
   parameterName=Curve
  }
  n5 = StaticComponentMaskParameter(None=AppendVector(A=@n4, B=@n4)) {defaultR=true}
  n6 = Multiply(
   A=DepthFade(Opacity=Multiply(A=@n3, B=@n5), FadeDistance=ScalarParameter() {defaultValue=20, parameterName=DF_FD}) {fadeDistanceDefault=20}
   B=ScalarParameter() {defaultValue=1, parameterName=Opacity_Multiply}
  )
  n7 = ParticleColor()
  LinearInterpolate(A=@n6, B=Multiply(A=@n6, B=@n7.A), Alpha=ScalarParameter() {parameterName=UseParticleAlpha, sliderMax=1})
  n8 = TextureCoordinate()
  n9 = ScalarParameter() {defaultValue=35, parameterName=Size}
  n10 = Multiply(A=VertexColor().R, B=ScalarParameter() {defaultValue=15, parameterName=SizeAdditive})
  n11 = MaterialFunctionCall(Float2=Multiply(A=Subtract(A=@n8) {constB=0.5}, B=Add(A=@n9, B=@n10))) {materialFunction=&BreakOutFloat2Components}
  n12 = Transform(None=Constant3Vector() {constant=<rgba 1 0 0 0>}) {transformSourceType=TRANSFORMSOURCE_Camera}
  n13 = Transform(None=Constant3Vector() {constant=<rgba 0 -1 0 0>}) {transformSourceType=TRANSFORMSOURCE_Camera}
  LinearInterpolate(
   A=Add(A=Multiply(A=@n11.R, B=@n12), B=Multiply(A=@n13, B=@n11.G))
   B=Constant()
   Alpha=ScalarParameter() {parameterName=LerpOutWPO, sliderMax=1}
  )
  n14 = VectorParameter() {parameterName=Color, defaultValue=<rgba 0.0293 0.117 0.2083 0>}
  MaterialFunctionCall(Emissive=LinearInterpolate(A=@n14.RGB, B=Multiply(A=@n7.RGB, B=@n14.RGB), Alpha=ScalarParameter() {parameterName=UseParticleColor, sliderMax=1})) {materialFunction=&MF_FX_EmissiveAdjustments}
 ]
}
