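# M_HgShader is an unlit, two-sided translucent effect material for the Space Bunny pickaxe, apparently a hologram.
# Vertex color R and G blend three layers: noise-displaced rays, a patterned body with intermittent glitch displacement, and that body undisplaced.
# Curve-atlas rows shape scrolling masks, rotated-UV triangle patterns, color and view-angle opacity.
# Master emissive and opacity controls follow, with depth fade skipped on mobile and a shared emissive-adjustment function last.
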Material /BRCosmetics/Weapons/FORT_Melee/Pickaxe_Space_Bunny/FX/M_HgShader
graph {
 properties {
  bUseMaterialAttributes=true
  blendMode=BLEND_Translucent
  pixelDepthOffsetMode=PDOM_Legacy
  shadingModel=MSM_Unlit
  twoSided=true
  usedShadingModels=Unlit
 }
 nodes [
  n1 = TextureCoordinate() {coordinateIndex=1}
  n2 = ComponentMask(None=@n1) {g=true}
  n3 = Time()
  n4 = Constant2Vector() {g=0.01, r=-1}
  n5 = TextureSample(UVs=Add(A=@n2, B=Multiply(A=@n3, B=@n4))) {texture=&AtmosphericCloudNoise01}
  n6 = CurveAtlasRowParameter(CurveTime=Add(A=@n5.R) {constB=0.2}) {
   atlas=&AthenaCurveAtlas01
   curve=/Game/Athena/LinearColorCurves/SpaceBunny_Curve_Color
   defaultValue=70
   parameterName=Color
  }
  n7 = Multiply(A=@n6, B=ScalarParameter() {defaultValue=1.5, parameterName=Intensity2}) {constB=0.2}
  n8 = Subtract(A=TextureCoordinate() {coordinateIndex=1, vTiling=0.5}, B=Multiply(A=Time()) {constB=0.5})
  n9 = CurveAtlasRowParameter(CurveTime=ComponentMask(None=Frac(None=@n8)) {g=true}) {
   atlas=&AthenaCurveAtlas01
   curve=/Game/Athena/LinearColorCurves/Curve_Masks2
   defaultValue=69
   parameterName=Param2
  }
  n10 = Constant() {r=0.5}
  n11 = QualitySwitch(Default=@n9.G, Low=@n10, Medium=@n10)
  n12 = TextureCoordinate() {coordinateIndex=1, vTiling=2}
  n13 = Add(A=ComponentMask(None=@n12) {g=true}, B=Multiply(A=Time(), B=Constant2Vector() {g=-0.05, r=1}))
  n14 = TextureSample!(UVs=@n13, Tex, Level, Apply View MipBias) {
   constMipValue=1
   mipValueMode=TMVM_MipLevel
   samplerType=SAMPLERTYPE_LinearGrayscale
   texture=/Game/Global/Textures/Default/Noise/T_SwirlNoise_N_height
  }
  n15 = Multiply(A=Subtract(A=@n14.R) {constB=0.5}, B=ComponentMask(None=@n12) {r=true})
  n16 = VertexNormalWS()
  n17 = OneMinus(None=ComponentMask(None=@n1) {r=true})
  n18 = Multiply(A=@n17, B=@n17)
  n19 = Multiply(
   A=Multiply(A=Multiply(A=@n18, B=@n18), B=LinearInterpolate(Alpha=@n5.R) {constA=0.15, constB=0.75})
   B=CurveAtlasRowParameter(CurveTime=@n2) {atlas=&AthenaCurveAtlas01, curve=/Game/Athena/LinearColorCurves/Curve_Masks, defaultValue=67}.R
  )
  n20 = TransformPosition(None=WorldPosition().XYZ) {transformSourceType=TRANSFORMPOSSOURCE_World}
  n21 = Constant3Vector() {constant=<rgba 1 0 0 0>}
  n22 = Constant3Vector() {constant=<rgba 0 1 0 0>}
  n23 = AppendVector(
   A=DotProduct(A=@n20, B=Transform(None=@n21) {transformSourceType=TRANSFORMSOURCE_Camera})
   B=DotProduct(A=@n20, B=Transform(None=@n22) {transformSourceType=TRANSFORMSOURCE_Camera})
  )
  n24 = TextureSample!(UVs=Panner(Coordinate=Divide(A=@n23) {constB=128}) {speedY=-0.05}, Tex, Apply View MipBias) {texture=&AtmosphericCloudNoise01}
  n25 = TextureSample(UVs=@n13) {samplerType=SAMPLERTYPE_LinearGrayscale, texture=&T_SwirlNoise_N_height}
  n26 = Multiply(A=@n19, B=Multiply(A=Add(A=@n24.R) {constB=0.3}, B=OneMinus(None=@n25.R)))
  n27 = TextureSample(UVs=Multiply(A=Time(), B=Constant2Vector() {g=-0.01, r=0.2})) {samplerType=SAMPLERTYPE_Normal, texture=/Game/Global/Textures/Default/Noise/64x64Noise}
  n28 = SetMaterialAttributes!(
   MaterialAttributes
   Emissive Color=Desaturation(None=Multiply(A=@n7, B=@n11), Fraction=Constant() {r=0.25})
   World Position Offset=Multiply(A=Multiply(A=@n15, B=@n16), B=ScalarParameter() {defaultValue=40, parameterName=Ray_WPO})
   Opacity=Multiply(
    A=ShadingPathSwitch(Default=@n26, Mobile=@n19) {desc=Migrated from FeatureLevelSwitch}
    B=LinearInterpolate(Alpha=@n27.R) {constA=1, constB=1.5}
   )
  ) {attributeSetTypes [#1, #2, #3]}
  n29 = Constant() {r=1}
  n30 = VertexColor()
  n31 = VectorParameter() {defaultValue=<rgba 12 20 8 16>}
  n32 = Multiply(A=@n30.R, B=AppendVector(A=@n31.R, B=@n31.R))
  n33 = Multiply(A=AppendVector(A=@n31.B, B=@n31.B), B=@n30.G)
  n34 = Add(A=Multiply(A=TextureCoordinate() {coordinateIndex=2}, B=Add(A=@n32, B=@n33)), B=Multiply(A=Time()) {constB=0.25})
  n35 = MaterialFunctionCall(UVs=@n34, "Rotation Angle (0-1)"=Constant() {r=0.3333}) {materialFunction=&CustomRotator}
  n36 = CurveAtlasRowParameter(CurveTime=ComponentMask(None=Frac(None=@n35.Rotated Values)) {r=true}) {
   atlas=&AthenaCurveAtlas01
   curve=&Curve_Masks
   defaultValue=67
   parameterName=Param_2
  }
  n37 = MaterialFunctionCall(UVs=@n34, "Rotation Angle (0-1)"=Constant() {r=0.1667}) {materialFunction=&CustomRotator}
  n38 = CurveAtlasRowParameter(CurveTime=ComponentMask(None=Frac(None=@n37.Rotated Values)) {r=true}) {
   atlas=&AthenaCurveAtlas01
   curve=&Curve_Masks
   defaultValue=67
   parameterName=Param_3
  }
  n39 = Add(A=@n36.G, B=@n38.G)
  n40 = CurveAtlasRowParameter(CurveTime=ComponentMask(None=Frac(None=@n34)) {r=true}) {
   atlas=&AthenaCurveAtlas01
   curve=&Curve_Masks
   defaultValue=67
   parameterName=Param_4
  }
  n41 = LinearInterpolate(A=ScalarParameter() {defaultValue=-0.25, parameterName=Triangle}, B=@n29, Alpha=Saturate(None=Add(A=@n39, B=@n40.G)))
  n42 = Add(A=ComponentMask(None=TextureCoordinate() {coordinateIndex=2}) {r=true}, B=Multiply(A=Time()) {constB=-0.2})
  n43 = CurveAtlasRowParameter(CurveTime=Frac(None=Multiply(A=@n42) {constB=6})) {
   atlas=&AthenaCurveAtlas01
   curve=/Game/Athena/LinearColorCurves/Curve_Masks1
   defaultValue=68
   parameterName=Param2
  }
  n44 = TextureCoordinate() {coordinateIndex=2}
  n45 = Multiply(
   A=@n43.R
   B=CurveAtlasRowParameter(CurveTime=ComponentMask(None=@n44) {r=true}) {atlas=&AthenaCurveAtlas01, curve=&Curve_Masks, defaultValue=67}.A
  )
  n46 = TextureSample(UVs=Multiply(A=Time(), B=Constant2Vector() {g=-0.01, r=0.2})) {samplerType=SAMPLERTYPE_Normal, texture=&64x64Noise}
  n47 = VertexColor()
  n48 = ComponentMask(None=TextureCoordinate() {coordinateIndex=2}) {g=true}
  n49 = CurveAtlasRowParameter(CurveTime=Multiply(A=@n47.G, B=@n48)) {
   atlas=&AthenaCurveAtlas01
   curve=&Curve_Masks2
   defaultValue=69
   parameterName=Param3
  }
  n50 = CurveAtlasRowParameter(CurveTime=Multiply(A=@n47.R, B=@n48)) {
   atlas=&AthenaCurveAtlas01
   curve=&Curve_Masks1
   defaultValue=68
   parameterName=Param2
  }
  n51 = CurveAtlasRowParameter(CurveTime=ComponentMask(None=TextureCoordinate() {coordinateIndex=2, vTiling=0.5}) {g=true}) {
   atlas=&AthenaCurveAtlas01
   curve=&Curve_Masks1
   defaultValue=68
   parameterName=Param2
  }
  n52 = Time()
  n53 = VertexColor()
  n54 = Add(A=Multiply(A=@n52, B=Multiply(A=@n53.R) {constB=-0.5}), B=Multiply(A=@n52, B=Multiply(A=@n53.G) {constB=-0.3143}))
  n55 = CurveAtlasRowParameter(CurveTime=Frac(None=Add(A=@n51.A, B=@n54))) {
   atlas=&AthenaCurveAtlas01
   curve=&Curve_Masks1
   defaultValue=68
   parameterName=Param2
  }
  n56 = Add(A=Multiply(A=LinearInterpolate(Alpha=@n46.G) {constA=0.75}, B=Min(A=@n49.B, B=@n50.B)), B=@n55.G)
  n57 = Constant()
  n58 = QualitySwitch(Default=@n56, Low=@n57, High=@n56, Medium=@n57)
  n59 = VertexColor()
  n60 = Add(A=@n59.R, B=Multiply(A=@n59.G) {constB=0.33})
  n61 = ScalarParameter() {defaultValue=0.3, parameterName=Bise}
  n62 = CurveAtlasRowParameter(CurveTime=Multiply(A=Add(A=@n58, B=@n45), B=Multiply(A=@n60, B=@n61))) {
   atlas=&AthenaCurveAtlas01
   curve=&SpaceBunny_Curve_Color
   defaultValue=70
   parameterName=Color
  }
  n63 = Multiply(A=Saturate(None=Subtract(A=@n58))) {constB=0.01}
  n64 = TextureSample() {samplerType=SAMPLERTYPE_Normal, texture=/BRCosmetics/Weapons/FORT_Melee/Pickaxe_Space_Bunny/Textures/T_Pickaxe_Space_Bunny_N}
  n65 = ConstantBiasScale(None=DotProduct(A=CameraVectorWS(), B=Transform(None=@n64.RGB)))
  n66 = Multiply(
   A=LinearInterpolate(Alpha=@n27.R) {constA=0.8, constB=1.5}
   B=CurveAtlasRowParameter(CurveTime=@n65) {atlas=&AthenaCurveAtlas01, curve=&Curve_Masks, defaultValue=67}.B
  )
  n67 = VertexColor()
  n68 = Multiply(A=@n65, B=@n65)
  n69 = SetMaterialAttributes!(
   MaterialAttributes
   Emissive Color=Multiply(A=ShadingPathSwitch(Default=@n41, Mobile=@n29) {desc=Migrated from FeatureLevelSwitch}, B=@n62)
   Opacity=Add(A=Add(A=@n63, B=@n66), B=Multiply(A=@n67.B, B=@n68) {constB=0.5})
  ) {attributeSetTypes [#1, #3]}
  n70 = TextureSample!(UVs=Add(A=TextureCoordinate() {coordinateIndex=2}, B=Time()), Tex, Level, Apply View MipBias) {constMipValue=1, mipValueMode=TMVM_MipLevel, texture=&AtmosphericCloudNoise01}
  n71 = Multiply(A=Multiply(A=VertexNormalWS(), B=Subtract(A=@n70.R) {constB=0.25}), B=ScalarParameter() {defaultValue=30, parameterName=Corrat_WPO})
  n72 = TextureSample(UVs=Multiply(A=Time(), B=Constant2Vector() {g=-0.017, r=0.05})) {mipValueMode=TMVM_MipLevel, samplerType=SAMPLERTYPE_Normal, texture=&64x64Noise}
  n73 = ScalarParameter() {defaultValue=0.85, parameterName=Ferquence}
  n74 = Saturate(None=Subtract(A=@n72.B, B=@n73) {constB=0.9})
  n75 = Time()
  n76 = Constant2Vector() {g=-0.01, r=0.2}
  n77 = Subtract(A=@n72.G, B=@n73) {constB=0.9}
  n78 = Multiply(
   A=TextureSample(UVs=Multiply(A=@n75, B=@n76)) {mipValueMode=TMVM_MipLevel, samplerType=SAMPLERTYPE_Normal, texture=&64x64Noise}.RGB
   B=Ceil(None=Saturate(None=@n77))
  )
  n79 = ScalarParameter() {defaultValue=0.8, parameterName=CorratPos_WPO}
  n80 = SetMaterialAttributes!(MaterialAttributes=@n69, World Position Offset=Add(A=Multiply(A=@n71, B=@n74), B=Multiply(A=@n78, B=@n79))) {attributeSetTypes [#2]}
  n81 = VertexColor()
  n82 = GetMaterialAttributes(None=BlendMaterialAttributes(A=BlendMaterialAttributes(A=@n28, B=@n80, Alpha=@n81.R), B=@n69, Alpha=@n81.G)) {attributeGetTypes [#1, #3]}
  n83 = SetMaterialAttributes!(
   MaterialAttributes=@n82.MaterialAttributes
   Emissive Color=Multiply(A=@n82.Emissive Color, B=ScalarParameter() {defaultValue=0.5, parameterName=Master_Emissive})
   Opacity=Multiply(A=@n82.Opacity, B=ScalarParameter() {defaultValue=1, parameterName=Master_Opacity})
  ) {attributeSetTypes [#1, #3]}
  n84 = GetMaterialAttributes(None=@n83) {attributeGetTypes [#3]}
  n85 = CurveAtlasRowParameter(CurveTime=DepthFade(FadeDistance=ScalarParameter() {defaultValue=20, parameterName=DF_FD})) {
   atlas=&AthenaCurveAtlas01
   curve=&Curve_Masks2
   defaultValue=69
   parameterName=Param3
  }
  n86 = Multiply(A=@n84.Opacity, B=@n85.A)
  MaterialFunctionCall!(In=ShadingPathSwitch(
    Default=SetMaterialAttributes!(MaterialAttributes=@n84.MaterialAttributes, Opacity=@n86) {attributeSetTypes [#3]}
    Mobile=@n83
   ) {desc=Migrated from FeatureLevelSwitch}) {materialFunction=/Game/Effects/Niagara/MasterMaterials/Functions/MF_FX_EmissiveAdjustments_MA}
 ]
}
